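# Conversation database for the HerbHutch NPC, providing its greeting and main conversation branches.
# The entry point leads to one Speech node whose contextual lines are gated by first-ever conversation, per-NPC conversation count or player outfit (CID).
# First-meeting and outfit lines use named priority curve rows.
# The speech branches through link nodes to separately tagged conversations: NPC quests, basic services, HerbHutch dialogue and a Season 31 story quest.
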ConversationDatabase /NPCLibrary/NPCs/HerbHutch/NPCLibrary_Conversation_HerbHutch
entry_points [{tag=Conversation.Entry.HerbHutch, destinations [#1]}]
nodes [
 ConversationEntryPointNode_7 {
  id=#1
  class=ConversationEntryPointNode
  outputs [#2]
  properties {entryTag {tagName=Conversation.Entry.HerbHutch}}
 }
 Speech_C_4 {
  id=#2
  class=Speech_C
  outputs [#3, #4, #5, #6]
  properties {
   nodeName=""
   generalConfig {
    defaultMessage=I think I'm done with vengeance. Time for some justice.
    contextualMessages [
     {
      message="I guess it's us vs. Doom, huh? Bring it on!"
      contextRequirements [{requirement {class=FortControllerRequirement_IsFirstEverConversationWithNPC}}]
      priority {curve {rowName=Tandem.Speech.Priority.FirstEverInteraction}}
     }
     {
      message=I think I'm done with vengeance. Time for some justice.
      contextRequirements [{requirement {class=FortControllerRequirement_CheckNPCMatchConversationCount, properties {nPCUniqueIdTag {tagName=AISpawnerData.Type.Tandem.HerbHutch}}}}]
      priority {curve {rowName=None}}
     }
     {
      message=I figured out that there's more to life than vengeance.
      contextRequirements [{requirement {class=FortControllerRequirement_CheckNPCMatchConversationCount, properties {nPCUniqueIdTag {tagName=AISpawnerData.Type.Tandem.HerbHutch}}}}]
      priority {curve {rowName=None}}
     }
     {
      message="It felt really good to be mad, for a long time. But I'm done with that."
      contextRequirements [{requirement {class=FortControllerRequirement_CheckNPCMatchConversationCount, properties {nPCUniqueIdTag {tagName=AISpawnerData.Type.Tandem.HerbHutch}}}}]
      priority {curve {rowName=None}}
     }
     {
      message="It's been a minute since I've had to face down an apocalypse. Ah, nostalgia..."
      contextRequirements [{requirement {class=FortControllerRequirement_CheckNPCMatchConversationCount, properties {nPCUniqueIdTag {tagName=AISpawnerData.Type.Tandem.HerbHutch}}}}]
      priority {curve {rowName=None}}
     }
     {
      message="You can call me Captain Jones. And check out... wait, where'd I put that shield...?"
      contextRequirements [{requirement {class=FortControllerRequirement_CheckNPCMatchConversationCount, properties {nPCUniqueIdTag {tagName=AISpawnerData.Type.Tandem.HerbHutch}}}}]
      priority {curve {rowName=None}}
     }
     {
      message="Hope's facing down her proper supervillain. I'm so proud!"
      contextRequirements [{requirement {class=FortControllerRequirement_CheckNPCMatchConversationCount, properties {nPCUniqueIdTag {tagName=AISpawnerData.Type.Tandem.HerbHutch}}}}]
      priority {curve {rowName=None}}
     }
     {
      message="You're doing great, kiddo. Keep up the fine work."
      contextRequirements [{requirement {
         class=FortControllerRequirement_HasCID
         properties {allowedCIds [
           &Character_VoidRedemption_Rebel
           &Character_BrightDisk
          ]}
        }}]
      priority {curve {rowName=Tandem.Speech.Priority.OutfitReaction}}
     }
     {
      message="See what 'a strong hand' gets you? Villains, every time. Duh."
      contextRequirements [{requirement {
         class=FortControllerRequirement_HasCID
         properties {allowedCIds [/BRCosmetics/Athena/Items/Cosmetics/Characters/Character_CeremonialGuard_Fencer]}
        }}]
      priority {curve {rowName=Tandem.Speech.Priority.OutfitReaction}}
     }
    ]
   }
  }
 }
 ConversationLinkNode_3 {
  id=#6
  class=ConversationLinkNode
  properties {remoteEntryTag {tagName=Conversation.Entry.Season31.Story.P02.Q07}}
 }
 ConversationLinkNode_2 {id=#5, class=ConversationLinkNode, properties {remoteEntryTag {tagName=Conversation.Entry.HerbHutch.Dialogue}}}
 ConversationLinkNode_0 {id=#4, class=ConversationLinkNode, properties {remoteEntryTag {tagName=Conversation.Entry.BasicServices}}}
 ConversationLinkNode_1 {id=#3, class=ConversationLinkNode, properties {remoteEntryTag {tagName=Conversation.Entry.NPCLibrary.Quests}}}
]
