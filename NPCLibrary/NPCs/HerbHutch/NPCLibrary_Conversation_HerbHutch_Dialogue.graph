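# Conversation database for the Herb Hutch NPC's "Ask about..." dialogue, entered through the Conversation.Entry.HerbHutch.Dialogue tag.
# An opening prompt branches into choices for Doom, Hope and The Oracle, plus a Back choice that exits the branch.
# Each topic plays a linear chain of Speech nodes with the NPC's opinions, and all three converge on one shared ReturnToStart BACK node.
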ConversationDatabase /NPCLibrary/NPCs/HerbHutch/NPCLibrary_Conversation_HerbHutch_Dialogue
entry_points [{tag=Conversation.Entry.HerbHutch.Dialogue, destinations [#1]}]
nodes [
 ConversationEntryPointNode_0 {
  id=#1
  class=ConversationEntryPointNode
  outputs [#2]
  properties {entryTag {tagName=Conversation.Entry.HerbHutch.Dialogue}}
 }
 Speech_C_0 {
  id=#2
  class=Speech_C
  outputs [#3, #4, #5, #6]
  properties {nodeName="", softTaskIcon=&T_UI_NPC_question, generalConfig {defaultMessage=What would ya like to know?}}
  sub_nodes [ConversationChoiceNode_5 {class=ConversationChoiceNode, properties {defaultChoiceDisplayText=Ask about...}}]
 }
 Speech_C_5 {
  id=#6
  class=Speech_C
  outputs [#7]
  properties {
   nodeName=""
   softTaskIcon=&T_UI_NPC_question
   generalConfig {defaultMessage="If I never get sassed by that sentient bowl of soup again, it'll be too soon."}
  }
  sub_nodes [ConversationChoiceNode_15 {class=ConversationChoiceNode, properties {defaultChoiceDisplayText=The Oracle}}]
 }
 Speech_C_6 {
  id=#7
  class=Speech_C
  outputs [#8]
  properties {
   nodeName=""
   softTaskIcon=&T_UI_NPC_question
   generalConfig {defaultMessage="That said, I respect her... even if her 'prophecies' don't tell us much."}
  }
 }
 Speech_C_8 {
  id=#8
  class=Speech_C
  outputs [#9]
  properties {
   nodeName=""
   softTaskIcon=&T_UI_NPC_question
   generalConfig {defaultMessage=At least THIS ally isn't going to betray us and try to shoot me. I hope.}
  }
 }
 Speech_C_9 {
  id=#9
  class=Speech_C
  outputs [#10]
  properties {
   nodeName=""
   softTaskIcon=&T_UI_NPC_question
   generalConfig {defaultMessage="Though honestly, I'd love to see her try."}
  }
 }
 ReturnToStart_C_23 {
  id=#10
  class=ReturnToStart_C
  properties {softTaskIcon=&T_UI_NPC_Back}
  sub_nodes [ConversationChoiceNode_32 {class=ConversationChoiceNode, properties {defaultChoiceDisplayText=BACK}}]
 }
 Speech_C_1 {
  id=#5
  class=Speech_C
  outputs [#11]
  properties {
   nodeName=""
   softTaskIcon=&T_UI_NPC_question
   generalConfig {defaultMessage="Hope's coming into her own as a hero, and I'm really proud of her."}
  }
  sub_nodes [ConversationChoiceNode_1 {class=ConversationChoiceNode, properties {defaultChoiceDisplayText=Hope}}]
 }
 Speech_C_2 {
  id=#11
  class=Speech_C
  outputs [#12]
  properties {
   nodeName=""
   softTaskIcon=&T_UI_NPC_question
   generalConfig {defaultMessage="I remember those pressures. Saving the world is a heavy burden to bear. "}
  }
 }
 Speech_C_3 {
  id=#12
  class=Speech_C
  outputs [#13]
  properties {
   nodeName=""
   softTaskIcon=&T_UI_NPC_question
   generalConfig {defaultMessage="I was like, \"Why me?\" \"I'm just a guy!\" \"Who are all these heroes flying around in hi-tech suits?"}
  }
 }
 Speech_C_4 {
  id=#13
  class=Speech_C
  outputs [#10]
  properties {
   nodeName=""
   softTaskIcon=&T_UI_NPC_question
   generalConfig {defaultMessage=She's ready. I know she is. And we'll have her back every step of the way.}
  }
 }
 Back_C_0 {
  id=#4
  class=Back_C
  sub_nodes [ConversationChoiceNode_7 {class=ConversationChoiceNode, properties {defaultChoiceDisplayText=BACK}}]
 }
 Speech_C_7 {
  id=#3
  class=Speech_C
  outputs [#14]
  properties {
   nodeName=""
   softTaskIcon=&T_UI_NPC_question
   generalConfig {defaultMessage="I can't say I'm a big fan of the guy, for obvious reasons."}
  }
  sub_nodes [ConversationChoiceNode_0 {class=ConversationChoiceNode, properties {defaultChoiceDisplayText=Doom}}]
 }
 Speech_C_15 {
  id=#14
  class=Speech_C
  outputs [#15]
  properties {
   nodeName=""
   softTaskIcon=&T_UI_NPC_question
   generalConfig {defaultMessage="He's not gonna win. I don't think. You always expect one of these to be your last, I guess."}
  }
 }
 Speech_C_14 {
  id=#15
  class=Speech_C
  outputs [#10]
  properties {
   nodeName=""
   softTaskIcon=&T_UI_NPC_question
   generalConfig {defaultMessage="But Hope and the Avengers and the whole team, have got that. I believe in us."}
  }
 }
]
